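# Material function that layers ice and snow over incoming Cattus body material attributes.
# A pre-skinned normal Z mask, computed per vertex, combines with a tiled snow texture through HeightLerp to form the ice coverage and snow-to-ice mix.
# That mix drives color, roughness and detail normal, and the layer blends in through coverage times the mask texture's red channel.
# ES2 and ES3_1 feature levels pass the input through unchanged.
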MaterialFunction /Game/Characters/Enemies/Cattus/Materials/Functions/MF_Cattus_Body_Ice
graph {nodes [
  n1 = FunctionInput() {inputName=Material In, inputType=FunctionInput_MaterialAttributes}
  n2 = VectorParameter() {group=Cattus_Ice, parameterName=Ice Color, defaultValue=<rgba 0.0672 0.304 0.6146 1>}
  n3 = VectorParameter() {group=Cattus_Ice, parameterName=Snow Color, defaultValue=<rgba 0.5563 0.6804 1.2 1>}
  n4 = Multiply(A=TextureCoordinate(), B=ScalarParameter() {defaultValue=3, group=Cattus_Ice, parameterName=Texture Tiling})
  n5 = TextureSampleParameter2D(UVs=@n4) {group=Cattus_Ice, parameterName=Tiling Overlay Texture, texture=/Game/Packages/Fortnite_UniqueMaterials_1/Textures/T_Ice_Lake_Snow_D}
  n6 = PreSkinnedNormal() {desc="outputs where the mesh normals\r\nare facing, but in vertex space"}
  n7 = ScalarParameter() {defaultValue=-0.35, desc=adjust angle at which forward facing normals output 0, group=Cattus_Ice, parameterName=Normals Mask Angle Adjust}
  n8 = Multiply(
   A=Add(A=ComponentMask(None=@n6) {b=true, desc=restricts to z axis}, B=@n7)
   B=ScalarParameter() {defaultValue=2, group=Cattus_Ice, parameterName=Normals Mask Contrast Adjust}
  )
  n9 = ComponentMask(None=VertexInterpolator(VS=Saturate(None=@n8)).PS) {desc="only outputs the first channel from vertex shader; also fixes red tint bug", r=true}
  n10 = Add(A=@n5.G, B=@n9)
  n11 = ScalarParameter() {defaultValue=0.8, desc=.8, group=Cattus_Ice, parameterName=Transition Phase Ice Mask Strength}
  n12 = MaterialFunctionCall(
   Transition Phase=Power(
    Base=Multiply(A=@n10, B=@n11)
    Exp=ScalarParameter() {defaultValue=0.65, desc=.5, group=Cattus_Ice, parameterName=Transition Phase Ice Mask Falloff}
   )
   Height Texture=@n5.G
  ) {materialFunction=&HeightLerp}
  n13 = ScalarParameter() {defaultValue=2, group=Cattus_Ice, parameterName=Snow to Ice Adjust}
  n14 = Saturate(None=Multiply(A=@n12.Alpha, B=Power(Base=@n9, Exp=@n13)))
  n15 = ScalarParameter() {defaultValue=0.07, group=Cattus_Ice, parameterName=Roughness Ice}
  n16 = ScalarParameter() {defaultValue=0.2143, group=Cattus_Ice, parameterName=Roughness Snow}
  n17 = ScalarParameter() {defaultValue=0.1, group=Cattus_Ice, parameterName=Ice Emissive Power}
  n18 = TextureSample(UVs=@n4) {samplerType=SAMPLERTYPE_Normal, texture=/Game/Packages/Fortnite_UniqueMaterials_1/Textures/T_WaterSoft_N}
  n19 = ScalarParameter() {defaultValue=-0.25, parameterName=Ice Normal flatness}
  n20 = Normalize(VectorInput=LinearInterpolate(
    A=Constant3Vector() {constant=<rgba 0 0 1 0>}
    B=MaterialFunctionCall(Normal=@n18.RGB, Flatness=@n19) {materialFunction=&FlattenNormal}.Result
    Alpha=@n14
   ))
  n21 = GetMaterialAttributes(None=@n1) {attributeGetTypes [#1]}
  n22 = TextureSampleParameter2D() {
   desc=Texture Masks
   parameterName=Texture Masks
   samplerType=SAMPLERTYPE_Masks
   texture=/Game/Characters/Enemies/DUDEBRO/Textures/T_DUDEBRO_Body_Body_M
  }
  n23 = MaterialFunctionCall!(
   Base Material=@n1
   Top Material=MakeMaterialAttributes!(
    BaseColor=LinearInterpolate(A=@n2.RGB, B=@n3.RGB, Alpha=@n14)
    Metallic
    Specular=Constant() {r=0.25}
    Roughness=LinearInterpolate(A=@n15, B=@n16, Alpha=@n14)
    Anisotropy
    EmissiveColor=Multiply(A=@n2.RGB, B=@n17)
    Opacity
    OpacityMask
    Normal=MaterialFunctionCall(BaseNormal=@n21.Normal, AdditionalNormal=@n20) {materialFunction=&BlendAngleCorrectedNormals}.Result
    Tangent
    WorldPositionOffset
    SubsurfaceColor
    ClearCoat
    ClearCoatRoughness
    AmbientOcclusion
    Refraction
    CustomizedUVs_0
    CustomizedUVs_1
    CustomizedUVs_2
    CustomizedUVs_3
    CustomizedUVs_4
    CustomizedUVs_5
    CustomizedUVs_6
    CustomizedUVs_7
    PixelDepthOffset
    ShadingModel
    Displacement
   )
   Alpha=Multiply(A=@n12.Alpha, B=@n22.R)
  ) {materialFunction=/Engine/Functions/MaterialLayerFunctions/MatLayerBlend_Standard}
  FunctionOutput(None=FeatureLevelSwitch(Default=@n23.Blended Material, ES2=@n1, ES3_1=@n1, SM4_REMOVED=@n23.Blended Material))
 ]}
