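# Material function for the Cattus death dissolve, turning the incoming material to ash, then eroding it away.
# Two sphere masks around an offset object position, perturbed by world-aligned cloud noise, follow Erode or a debug sine.
# The larger blends in tinted ash using paint textures; the smaller drives dithered opacity and an optional glowing edge.
# A switchable noise-driven world position offset branch exists but is not wired to the output.
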MaterialFunction /Game/Effects/Fort_Effects/Events/CattusDoggus/Cattus/MF_Cattus_DeathDissolve
graph {nodes [
  n1 = VectorParameter() {parameterName=AshBaseColor, defaultValue=<rgba 0.0451 0.05 0.0189 1>}
  n2 = LinearInterpolate(
   A=MaterialFunctionCall(Hue Shift Percentage=Constant() {r=0.05}, Texture=@n1.RGB) {materialFunction=&HueShift}.Result
   B=@n1.RGB
   Alpha=TextureSample(UVs=TextureCoordinate() {uTiling=5, vTiling=5}) {texture=/Game/Effects/Fort_Effects/Textures/T_Tile_Pattern_01}.R
  )
  n3 = TextureSampleParameter2D() {
   group=Paint
   parameterName=S Mask
   samplerType=SAMPLERTYPE_Masks
   sortPriority=5
   texture=/Game/Characters/Enemies/Cattus/Textures/T_Cattus_Paint_Head_S
  }
  n4 = Add(A=@n3.B) {constB=0.3}
  n5 = TextureSampleParameter2D() {
   group=Paint
   parameterName=Normal
   samplerType=SAMPLERTYPE_Normal
   sortPriority=4
   texture=/Game/Characters/Enemies/Cattus/Textures/T_Cattus_Paint_Head_N
  }
  n6 = WorldPosition()
  n7 = VectorParameter() {parameterName=PointOffset, defaultValue=<rgba 6000 -5500 0 1>}
  n8 = MaterialFunctionCall(
   TextureObject=TextureObject() {texture=/Game/Effects/Fort_Effects/Textures/Tile/T_CloudScape_08}
   TextureSize=VectorParameter() {parameterName=TectureSize, defaultValue=<rgba 4000 3000 3000 0>}.RGB
   WorldPosition=WorldPosition().XYZ
  ) {materialFunction=&WorldAlignedTexture}
  n9 = Add(
   A=Add(A=ObjectPositionWS(), B=ComponentMask(None=@n7.RGB) {b=true, g=true, r=true})
   B=LinearInterpolate(
    A=ScalarParameter() {defaultValue=-2000, parameterName=NoiseRangeLow}
    B=ScalarParameter() {defaultValue=5000, parameterName=NoiseRangeHigh}
    Alpha=ComponentMask(None=@n8.XYZ Texture) {r=true}
   )
  )
  n10 = ScalarParameter() {defaultValue=25000, parameterName=ErodeRadius}
  n11 = ScalarParameter() {defaultValue=20000, parameterName=AshMaterialRadius}
  n12 = OneMinus(None=SphereMask(A=@n6.XYZ, B=@n9, Radius=Add(A=@n10, B=@n11) {constB=5000}) {hardnessPercent=0})
  n13 = ScalarParameter() {defaultValue=12, parameterName=ErodeSoftness}
  n14 = ScalarParameter() {defaultValue=0.05, parameterName=DebugTimeSpeed}
  n15 = LinearInterpolate(
   A=ScalarParameter() {defaultValue=-1, parameterName=ErodeRangeLow}
   B=ScalarParameter() {defaultValue=11, parameterName=ErodeRangeHight}
   Alpha=StaticSwitchParameter(
    True=MaterialFunctionCall(Speed=@n14) {materialFunction=&DebugTimeSine}.Result
    False=ScalarParameter() {parameterName=Erode}
   ) {parameterName=DebugErode}
  ) {constB=2}
  n16 = Saturate(None=Subtract(A=Multiply(A=@n12, B=@n13) {constB=5}, B=@n15))
  n17 = OneMinus(None=SphereMask(A=@n6.XYZ, B=@n9, Radius=@n10) {hardnessPercent=0})
  n18 = Saturate(None=Subtract(A=Multiply(A=@n17, B=@n13) {constB=5}, B=@n15))
  n19 = MaterialFunctionCall!(
   Material=BlendMaterialAttributes(
    A=MakeMaterialAttributes!(
     BaseColor=@n2
     Metallic=@n3.G
     Specular=@n3.G
     Roughness=@n4
     Anisotropy
     EmissiveColor
     Opacity
     OpacityMask
     Normal=@n5.RGB
     Tangent
     WorldPositionOffset
     SubsurfaceColor
     ClearCoat
     ClearCoatRoughness
     AmbientOcclusion
     Refraction
     CustomizedUVs_0
     CustomizedUVs_1
     CustomizedUVs_2
     CustomizedUVs_3
     CustomizedUVs_4
     CustomizedUVs_5
     CustomizedUVs_6
     CustomizedUVs_7
     PixelDepthOffset
     ShadingModel
     Displacement
    )
    B=FunctionInput() {inputType=FunctionInput_MaterialAttributes}
    Alpha=Power(Base=@n16)
   )
   NewOpacityMask=MaterialFunctionCall(Alpha Threshold=@n18) {materialFunction=&DitherTemporalAA}.Result
   Mask
  ) {materialFunction=/Engine/Functions/MaterialLayerFunctions/MatLayerBlend_OverrideOpacityMask}
  n20 = OneMinus(None=@n18)
  n21 = VectorParameter() {parameterName=ErodeEdgeColor, defaultValue=<rgba 45.8898 50 1.6 1>}
  n22 = StaticSwitchParameter(
   True=MaterialFunctionCall!(Material=@n19.Blended Material, Emissive=Multiply(A=@n20, B=@n21.RGB)) {materialFunction=/Engine/Functions/MaterialLayerFunctions/MatLayerBlend_Emissive}.Blended Material
   False=@n19.Blended Material
  ) {parameterName=UseGlowErodeEdge}
  FunctionOutput(None=@n22)
  n23 = VectorParameter() {parameterName=WPO_TilingAnim_Tiling, defaultValue=<rgba 1 1 0 0>}
  n24 = Panner(Coordinate=Multiply(A=TextureCoordinate(), B=AppendVector(A=@n23.R, B=@n23.R)), Speed=AppendVector(A=@n23.B, B=@n23.B)) {speedY=1}
  n25 = Multiply(
   A=Multiply(
    A=VertexNormalWS()
    B=TextureSample(UVs=@n24) {samplerType=SAMPLERTYPE_LinearColor, texture=&HighContrastNoise}.RGB
   )
   B=ScalarParameter() {parameterName=WPO_Intensity}
  )
  n26 = AppendVector(A=Constant2Vector(), B=Multiply(A=Constant() {r=-1}, B=ScalarParameter() {parameterName=WPO_Falling}))
  StaticSwitchParameter(
   True=MaterialFunctionCall!(Material=@n22, WorldPositionOffset=Multiply(A=@n25, B=@n26)) {materialFunction=/Engine/Functions/MaterialLayerFunctions/MatLayerBlend_AddWorldPositionOffset}.Blended Material
   False=@n22
  ) {parameterName=UseWPO_TilingAnim}
 ]}
