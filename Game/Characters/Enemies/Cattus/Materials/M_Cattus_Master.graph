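# Master masked, subsurface-profile material for the skeletal Cattus enemy, built on material attributes.
# Paint textures supply base color, normal and masks, with S Mask green shared by metallic, specular and roughness, and M Mask driving AO and subsurface color.
# Attributes pass through borrowed Doggus wetness, an optional frozen overlay, Dudebro skin FX, then death dissolve.
# The UseSkinFX static switch bypasses that entire chain when off, including wetness and ice.
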Material /Game/Characters/Enemies/Cattus/Materials/M_Cattus_Master
graph {
 properties {
  bUseMaterialAttributes=true
  bUsedWithMeshDeformer=true
  bUsedWithSkeletalMesh=true
  blendMode=BLEND_Masked
  pixelDepthOffsetMode=PDOM_Legacy
  shadingModel=MSM_FromMaterialExpression
  subsurfaceProfile=&/SSP_Cattus
  usedShadingModels=Subsurface Profile
 }
 nodes [
  n1 = TextureSampleParameter2D() {group=Paint, parameterName=Diffuse, sortPriority=1, texture=/Game/Characters/Enemies/Cattus/Textures/T_Cattus_Paint_Head_D}
  n2 = TextureSampleParameter2D() {
   group=Paint
   parameterName=S Mask
   samplerType=SAMPLERTYPE_Masks
   sortPriority=5
   texture=/Game/Characters/Enemies/Cattus/Textures/T_Cattus_Paint_Head_S
  }
  n3 = TextureSampleParameter2D() {
   group=Paint
   parameterName=Normal
   samplerType=SAMPLERTYPE_Normal
   sortPriority=4
   texture=/Game/Characters/Enemies/Cattus/Textures/T_Cattus_Paint_Head_N
  }
  n4 = TextureSampleParameter2D() {
   group=Paint
   parameterName=M Mask
   samplerType=SAMPLERTYPE_Masks
   sortPriority=2
   texture=/Game/Characters/Enemies/Cattus/Textures/T_Cattus_Paint_Head_M
  }
  n5 = MakeMaterialAttributes!(
   BaseColor=@n1.RGB
   Metallic=@n2.G
   Specular=@n2.G
   Roughness=@n2.G
   Anisotropy
   EmissiveColor
   Opacity=ScalarParameter() {defaultValue=1, group=Paint, parameterName=SS Opacity}
   OpacityMask
   Normal=@n3.RGB
   Tangent
   WorldPositionOffset
   SubsurfaceColor=LinearInterpolate(B=@n1.RGB, Alpha=@n4.B)
   ClearCoat
   ClearCoatRoughness
   AmbientOcclusion=LinearInterpolate(A=Constant() {r=1}, B=@n4.R, Alpha=ScalarParameter() {defaultValue=0.5, group=Paint, parameterName=AO Strength, sortPriority=3})
   Refraction
   CustomizedUVs_0
   CustomizedUVs_1
   CustomizedUVs_2
   CustomizedUVs_3
   CustomizedUVs_4
   CustomizedUVs_5
   CustomizedUVs_6
   CustomizedUVs_7
   PixelDepthOffset
   ShadingModel=ShadingModel() {shadingModel=MSM_SubsurfaceProfile}
   Displacement
  )
  n6 = MaterialFunctionCall!(AO_Mask_Input=@n4.R, In=@n5) {materialFunction=/Game/Characters/Player/Male/Medium/Bodies/M_MED_Doggus/MaterialSet/MF_DoggusWet}
  n7 = StaticSwitchParameter(
   True=MaterialFunctionCall!(Material In=@n6.Result) {materialFunction=/Game/Characters/Enemies/Cattus/Materials/Functions/MF_Cattus_Body_Ice}.Result
   False=@n6.Result
  ) {parameterName=UseFrozenOverlay}
  StaticSwitchParameter(
   True=MaterialFunctionCall!(In=MaterialFunctionCall!(InMA=@n7, HaveMasks, "FresnelRevert(WithOutMask)", EmissiveMutiply) {materialFunction=/Game/Characters/Enemies/DUDEBRO/FX/SkinFX/MF_DudebroSkinFX}.MA) {materialFunction=/Game/Effects/Fort_Effects/Events/CattusDoggus/Cattus/MF_Cattus_DeathDissolve}.Result
   False=@n5
  ) {parameterName=UseSkinFX}
 ]
}
